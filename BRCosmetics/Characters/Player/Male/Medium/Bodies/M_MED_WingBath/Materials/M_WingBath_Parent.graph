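# Parent material for the WingBath medium male player body cosmetic, letting variants retint and tune metal, cloth, leather, decal and sigil surfaces.
# Character PreFX output is split into layers with tint and metallic, specular and roughness biases, one adding hue shift and fresnel.
# Masks blend the layers into PostFX, with RGB mask edges widened by the max of four texel-offset samples.
# Sigils get parameterised offset, rotation and scale placements.
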Material /BRCosmetics/Characters/Player/Male/Medium/Bodies/M_MED_WingBath/Materials/M_WingBath_Parent
graph {
 expressions {
  MaterialExpressionAbs_0 {class=MaterialExpressionAbs, inputs [None {from=MaterialExpressionSubtract_0, output=""}], properties {}}
  MaterialExpressionAdd_0 {
   class=MaterialExpressionAdd
   inputs [
    A {from=MaterialExpressionGetMaterialAttributes_0, output=Metallic}
    B {from=MaterialExpressionScalarParameter_2}
   ]
  }
  MaterialExpressionAdd_1 {
   class=MaterialExpressionAdd
   inputs [
    A {from=MaterialExpressionGetMaterialAttributes_0, output=Specular}
    B {from=MaterialExpressionScalarParameter_1}
   ]
  }
  MaterialExpressionAdd_10 {
   class=MaterialExpressionAdd
   inputs [
    A {from=MaterialExpressionGetMaterialAttributes_3, output=Specular}
    B {from=MaterialExpressionScalarParameter_10}
   ]
  }
  MaterialExpressionAdd_11 {
   class=MaterialExpressionAdd
   inputs [
    A {from=MaterialExpressionGetMaterialAttributes_3, output=Roughness}
    B {from=MaterialExpressionAdd_23}
   ]
  }
  MaterialExpressionAdd_12 {
   class=MaterialExpressionAdd
   inputs [
    A {from=MaterialExpressionGetMaterialAttributes_4, output=Metallic}
    B {from=MaterialExpressionAdd_25}
   ]
  }
  MaterialExpressionAdd_13 {
   class=MaterialExpressionAdd
   inputs [
    A {from=MaterialExpressionGetMaterialAttributes_4, output=Specular}
    B {from=MaterialExpressionScalarParameter_13}
   ]
  }
  MaterialExpressionAdd_14 {
   class=MaterialExpressionAdd
   inputs [
    A {from=MaterialExpressionGetMaterialAttributes_4, output=Roughness}
    B {from=MaterialExpressionScalarParameter_12}
   ]
  }
  MaterialExpressionAdd_15 {
   class=MaterialExpressionAdd
   inputs [
    A {from=MaterialExpressionTextureCoordinate_1}
    B {from=MaterialExpressionAppendVector_0}
   ]
   properties {constB=0.1}
  }
  MaterialExpressionAdd_16 {
   class=MaterialExpressionAdd
   inputs [
    A {from=MaterialExpressionTextureCoordinate_2}
    B {from=MaterialExpressionAppendVector_1}
   ]
   properties {constB=-0.1}
  }
  MaterialExpressionAdd_17 {
   class=MaterialExpressionAdd
   inputs [
    A {from=MaterialExpressionTextureCoordinate_3}
    B {from=MaterialExpressionAppendVector_2}
   ]
   properties {constB=-0.1}
  }
  MaterialExpressionAdd_18 {
   class=MaterialExpressionAdd
   inputs [
    A {from=MaterialExpressionTextureCoordinate_4}
    B {from=MaterialExpressionAppendVector_3}
   ]
   properties {constB=-0.1}
  }
  MaterialExpressionAdd_19 {
   class=MaterialExpressionAdd
   inputs [
    A {from=MaterialExpressionTextureCoordinate_0}
    B {from=MaterialExpressionAppendVector_4}
   ]
  }
  MaterialExpressionAdd_2 {
   class=MaterialExpressionAdd
   inputs [
    A {from=MaterialExpressionGetMaterialAttributes_0, output=Roughness}
    B {from=MaterialExpressionScalarParameter_0}
   ]
  }
  MaterialExpressionAdd_20 {
   class=MaterialExpressionAdd
   inputs [
    A {from=MaterialExpressionTextureCoordinate_6}
    B {from=MaterialExpressionAppendVector_5}
   ]
  }
  MaterialExpressionAdd_21 {
   class=MaterialExpressionAdd
   inputs [
    A {from=MaterialExpressionTextureCoordinate_7}
    B {from=MaterialExpressionAppendVector_6}
   ]
  }
  MaterialExpressionAdd_22 {
   class=MaterialExpressionAdd
   inputs [
    A {from=MaterialExpressionTextureCoordinate_8}
    B {from=MaterialExpressionAppendVector_7}
   ]
  }
  MaterialExpressionAdd_23 {
   class=MaterialExpressionAdd
   inputs [
    A {from=MaterialExpressionScalarParameter_9}
    B {from=""}
   ]
   properties {constB=0.15}
  }
  MaterialExpressionAdd_24 {
   class=MaterialExpressionAdd
   inputs [
    A {from=MaterialExpressionTextureCoordinate_9}
    B {from=MaterialExpressionAppendVector_8}
   ]
  }
  MaterialExpressionAdd_25 {
   class=MaterialExpressionAdd
   inputs [
    A {from=MaterialExpressionScalarParameter_14}
    B {from=""}
   ]
   properties {constB=-0.35}
  }
  MaterialExpressionAdd_26 {
   class=MaterialExpressionAdd
   inputs [
    A {from=MaterialExpressionScalarParameter_11}
    B {from=MaterialExpressionConstant_4}
   ]
  }
  MaterialExpressionAdd_27 {
   class=MaterialExpressionAdd
   inputs [
    A {from=MaterialExpressionTextureCoordinate_10}
    B {from=MaterialExpressionAppendVector_11}
   ]
  }
  MaterialExpressionAdd_28 {
   class=MaterialExpressionAdd
   inputs [
    A {from=MaterialExpressionGetMaterialAttributes_5, output=Metallic}
    B {from=MaterialExpressionScalarParameter_57}
   ]
  }
  MaterialExpressionAdd_29 {
   class=MaterialExpressionAdd
   inputs [
    A {from=MaterialExpressionGetMaterialAttributes_5, output=Specular}
    B {from=MaterialExpressionScalarParameter_56}
   ]
  }
  MaterialExpressionAdd_3 {
   class=MaterialExpressionAdd
   inputs [
    A {from=MaterialExpressionGetMaterialAttributes_1, output=Metallic}
    B {from=MaterialExpressionScalarParameter_5}
   ]
  }
  MaterialExpressionAdd_30 {
   class=MaterialExpressionAdd
   inputs [
    A {from=MaterialExpressionGetMaterialAttributes_5, output=Roughness}
    B {from=MaterialExpressionScalarParameter_55}
   ]
  }
  MaterialExpressionAdd_4 {
   class=MaterialExpressionAdd
   inputs [
    A {from=MaterialExpressionGetMaterialAttributes_1, output=Specular}
    B {from=MaterialExpressionScalarParameter_4}
   ]
  }
  MaterialExpressionAdd_5 {
   class=MaterialExpressionAdd
   inputs [
    A {from=MaterialExpressionGetMaterialAttributes_1, output=Roughness}
    B {from=MaterialExpressionScalarParameter_3}
   ]
  }
  MaterialExpressionAdd_6 {
   class=MaterialExpressionAdd
   inputs [
    A {from=MaterialExpressionGetMaterialAttributes_2, output=Metallic}
    B {from=MaterialExpressionScalarParameter_8}
   ]
  }
  MaterialExpressionAdd_7 {
   class=MaterialExpressionAdd
   inputs [
    A {from=MaterialExpressionGetMaterialAttributes_2, output=Specular}
    B {from=MaterialExpressionScalarParameter_7}
   ]
  }
  MaterialExpressionAdd_8 {
   class=MaterialExpressionAdd
   inputs [
    A {from=MaterialExpressionGetMaterialAttributes_2, output=Roughness}
    B {from=MaterialExpressionScalarParameter_6}
   ]
  }
  MaterialExpressionAdd_9 {
   class=MaterialExpressionAdd
   inputs [
    A {from=MaterialExpressionGetMaterialAttributes_3, output=Metallic}
    B {from=MaterialExpressionAdd_26}
   ]
  }
  MaterialExpressionAppendVector_0 {
   class=MaterialExpressionAppendVector
   inputs [
    A {from=MaterialExpressionMultiply_10, output=""}
    B {from=MaterialExpressionConstant_1, output=""}
   ]
  }
  MaterialExpressionAppendVector_1 {
   class=MaterialExpressionAppendVector
   inputs [
    A {from=MaterialExpressionConstant_1, output=""}
    B {from=MaterialExpressionMultiply_10, output=""}
   ]
  }
  MaterialExpressionAppendVector_10 {
   class=MaterialExpressionAppendVector
   inputs [
    A {from=MaterialExpressionScalarParameter_47, output=""}
    B {from=MaterialExpressionScalarParameter_48, output=""}
   ]
  }
  MaterialExpressionAppendVector_11 {
   class=MaterialExpressionAppendVector
   inputs [
    A {from=MaterialExpressionScalarParameter_49, output=""}
    B {from=MaterialExpressionScalarParameter_50, output=""}
   ]
  }
  MaterialExpressionAppendVector_12 {
   class=MaterialExpressionAppendVector
   inputs [
    A {from=MaterialExpressionScalarParameter_53, output=""}
    B {from=MaterialExpressionScalarParameter_54, output=""}
   ]
  }
  MaterialExpressionAppendVector_2 {
   class=MaterialExpressionAppendVector
   inputs [
    A {from=MaterialExpressionMultiply_9, output=""}
    B {from=MaterialExpressionConstant_2, output=""}
   ]
  }
  MaterialExpressionAppendVector_3 {
   class=MaterialExpressionAppendVector
   inputs [
    A {from=MaterialExpressionConstant_3, output=""}
    B {from=MaterialExpressionMultiply_9, output=""}
   ]
  }
  MaterialExpressionAppendVector_4 {
   class=MaterialExpressionAppendVector
   inputs [
    A {from=MaterialExpressionScalarParameter_16, output=""}
    B {from=MaterialExpressionScalarParameter_18, output=""}
   ]
  }
  MaterialExpressionAppendVector_5 {
   class=MaterialExpressionAppendVector
   inputs [
    A {from=MaterialExpressionScalarParameter_21, output=""}
    B {from=MaterialExpressionScalarParameter_22, output=""}
   ]
  }
  MaterialExpressionAppendVector_6 {
   class=MaterialExpressionAppendVector
   inputs [
    A {from=MaterialExpressionScalarParameter_25, output=""}
    B {from=MaterialExpressionScalarParameter_26, output=""}
   ]
  }
  MaterialExpressionAppendVector_7 {
   class=MaterialExpressionAppendVector
   inputs [
    A {from=MaterialExpressionScalarParameter_29, output=""}
    B {from=MaterialExpressionScalarParameter_30, output=""}
   ]
  }
  MaterialExpressionAppendVector_8 {
   class=MaterialExpressionAppendVector
   inputs [
    A {from=MaterialExpressionScalarParameter_34, output=""}
    B {from=MaterialExpressionScalarParameter_35, output=""}
   ]
  }
  MaterialExpressionAppendVector_9 {
   class=MaterialExpressionAppendVector
   inputs [
    A {from=MaterialExpressionScalarParameter_20, output=""}
    B {from=MaterialExpressionScalarParameter_45, output=""}
   ]
  }
  MaterialExpressionBlendMaterialAttributes_0 {
   class=MaterialExpressionBlendMaterialAttributes
   inputs [
    A {from=MaterialExpressionNamedRerouteUsage_1, output=""}
    B {from=MaterialExpressionNamedRerouteUsage_2, output=""}
    Alpha {from=MaterialExpressionMaterialFunctionCall_2, output=R}
   ]
   properties {vertexAttributeBlendType=UseA}
  }
  MaterialExpressionBlendMaterialAttributes_1 {
   class=MaterialExpressionBlendMaterialAttributes
   inputs [
    A {from=MaterialExpressionBlendMaterialAttributes_0, output=""}
    B {from=MaterialExpressionNamedRerouteUsage_5, output=""}
    Alpha {from=MaterialExpressionMaterialFunctionCall_2, output=G}
   ]
   properties {vertexAttributeBlendType=UseA}
  }
  MaterialExpressionBlendMaterialAttributes_2 {
   class=MaterialExpressionBlendMaterialAttributes
   inputs [
    A {from=MaterialExpressionBlendMaterialAttributes_1, output=""}
    B {from=MaterialExpressionNamedRerouteUsage_6, output=""}
    Alpha {from=MaterialExpressionMaterialFunctionCall_2, output=B}
   ]
   properties {vertexAttributeBlendType=UseA}
  }
  MaterialExpressionBlendMaterialAttributes_3 {
   class=MaterialExpressionBlendMaterialAttributes
   inputs [
    A {from=MaterialExpressionBlendMaterialAttributes_5, output=""}
    B {from=MaterialExpressionNamedRerouteUsage_8, output=""}
    Alpha {from=MaterialExpressionSaturate_4, output=""}
   ]
   properties {vertexAttributeBlendType=UseA}
  }
  MaterialExpressionBlendMaterialAttributes_4 {
   class=MaterialExpressionBlendMaterialAttributes
   inputs [
    A {from=MaterialExpressionBlendMaterialAttributes_3, output=""}
    B {from=MaterialExpressionNamedRerouteUsage_11, output=""}
    Alpha {from=MaterialExpressionReroute_0, output=""}
   ]
   properties {vertexAttributeBlendType=UseA}
  }
  MaterialExpressionBlendMaterialAttributes_5 {
   class=MaterialExpressionBlendMaterialAttributes
   inputs [
    A {from=MaterialExpressionBlendMaterialAttributes_2, output=""}
    B {from=MaterialExpressionNamedRerouteUsage_14, output=""}
    Alpha {from=MaterialExpressionTextureSampleParameter2D_1, output=G}
   ]
   properties {vertexAttributeBlendType=UseA}
  }
  MaterialExpressionConstant_0 {class=MaterialExpressionConstant, properties {r=1}}
  MaterialExpressionConstant_1 {class=MaterialExpressionConstant}
  MaterialExpressionConstant_11 {class=MaterialExpressionConstant, properties {r=0.6}}
  MaterialExpressionConstant_12 {class=MaterialExpressionConstant, properties {r=0.6}}
  MaterialExpressionConstant_13 {class=MaterialExpressionConstant, properties {r=1}}
  MaterialExpressionConstant_2 {class=MaterialExpressionConstant}
  MaterialExpressionConstant_3 {class=MaterialExpressionConstant}
  MaterialExpressionConstant_4 {class=MaterialExpressionConstant, properties {r=-0.35}}
  MaterialExpressionConstant_5 {class=MaterialExpressionConstant, properties {r=0.2}}
  MaterialExpressionConstant_6 {class=MaterialExpressionConstant, properties {r=12}}
  MaterialExpressionConstant_8 {class=MaterialExpressionConstant, properties {r=1}}
  MaterialExpressionConstant_9 {class=MaterialExpressionConstant, properties {r=0.6}}
  MaterialExpressionConvert_0 {
   class=MaterialExpressionConvert
   inputs [None {from=MaterialExpressionNamedRerouteUsage_9}]
   properties {
    nodeName=Break Float3
    convertInputs [{
      type=Vector3
      expressionInput {
       expression=&self:MaterialExpressionNamedRerouteUsage_9
       inputName=None
       mask=0
       maskA=0
       maskB=0
       maskG=0
       maskR=0
       outputIndex=0
      }
     }]
    convertOutputs [{type=Scalar}, {type=Scalar}, {type=Scalar}]
   }
  }
  MaterialExpressionDesaturation_0 {
   class=MaterialExpressionDesaturation
   inputs [
    None {from=MaterialExpressionGetMaterialAttributes_5, output=Base Color}
    Fraction {from=MaterialExpressionScalarParameter_58, output=""}
   ]
   properties {}
  }
  MaterialExpressionDivide_0 {
   class=MaterialExpressionDivide
   inputs [
    A {from=MaterialExpressionConstant_8}
    B {from=""}
   ]
   properties {constB=2048}
  }
  MaterialExpressionFresnel_0 {
   class=MaterialExpressionFresnel
   inputs [
    ExponentIn {from=MaterialExpressionScalarParameter_61, output=""}
    BaseReflectFractionIn {from=MaterialExpressionScalarParameter_62, output=""}
    Normal {from="", output=""}
   ]
   properties {}
  }
  MaterialExpressionGetMaterialAttributes_0 {
   class=MaterialExpressionGetMaterialAttributes
   inputs [None {from=MaterialExpressionNamedRerouteUsage_0, output=""}]
   properties {attributeGetTypes [
     #1
     #2
     #3
     #4
    ]}
  }
  MaterialExpressionGetMaterialAttributes_1 {
   class=MaterialExpressionGetMaterialAttributes
   inputs [None {from=MaterialExpressionNamedRerouteUsage_3, output=""}]
   properties {attributeGetTypes [
     #1
     #2
     #3
     #4
    ]}
  }
  MaterialExpressionGetMaterialAttributes_2 {
   class=MaterialExpressionGetMaterialAttributes
   inputs [None {from=MaterialExpressionNamedRerouteUsage_4, output=""}]
   properties {attributeGetTypes [
     #1
     #2
     #3
     #4
    ]}
  }
  MaterialExpressionGetMaterialAttributes_3 {
   class=MaterialExpressionGetMaterialAttributes
   inputs [None {from=MaterialExpressionNamedRerouteUsage_7, output=""}]
   properties {attributeGetTypes [
     #1
     #2
     #3
     #4
    ]}
  }
  MaterialExpressionGetMaterialAttributes_4 {
   class=MaterialExpressionGetMaterialAttributes
   inputs [None {from=MaterialExpressionNamedRerouteUsage_10, output=""}]
   properties {attributeGetTypes [
     #1
     #2
     #3
     #4
    ]}
  }
  MaterialExpressionGetMaterialAttributes_5 {
   class=MaterialExpressionGetMaterialAttributes
   inputs [None {from=MaterialExpressionNamedRerouteUsage_12, output=""}]
   properties {attributeGetTypes [
     #1
     #2
     #3
     #4
    ]}
  }
  MaterialExpressionIf_0 {
   class=MaterialExpressionIf
   inputs [
    A {from=MaterialExpressionScalarParameter_39, output=""}
    B {from="", output=""}
    "A > B" {from=MaterialExpressionScalarParameter_40, output=""}
    "A == B" {from=MaterialExpressionConstant_13, output=""}
    "A < B" {from=MaterialExpressionConstant_13, output=""}
   ]
   properties {constB=1}
  }
  MaterialExpressionIf_1 {
   class=MaterialExpressionIf
   inputs [
    A {from=MaterialExpressionAbs_0, output=""}
    B {from="", output=""}
    "A > B" {from=MaterialExpressionVectorParameter_4, output=RGB}
    "A == B" {from=MaterialExpressionMultiply_14, output=""}
    "A < B" {from=MaterialExpressionMultiply_14, output=""}
   ]
   properties {constB=0.01}
  }
  MaterialExpressionLinearInterpolate_0 {
   class=MaterialExpressionLinearInterpolate
   inputs [
    A {from=MaterialExpressionMaterialFunctionCall_0, output=Tech Art Mask}
    B {from=MaterialExpressionNamedRerouteDeclaration_6}
    Alpha {from=""}
   ]
   properties {constAlpha=1}
  }
  MaterialExpressionLinearInterpolate_1 {
   class=MaterialExpressionLinearInterpolate
   inputs [
    A {from=MaterialExpressionScalarParameter_15}
    B {from=MaterialExpressionScalarParameter_17}
    Alpha {from=MaterialExpressionTextureSample_0, output=R}
   ]
   properties {constA=0.5}
  }
  MaterialExpressionLinearInterpolate_2 {
   class=MaterialExpressionLinearInterpolate
   inputs [
    A {from=MaterialExpressionMultiply_20}
    B {from=MaterialExpressionMultiply_21}
    Alpha {from=MaterialExpressionSaturate_3}
   ]
  }
  MaterialExpressionLinearInterpolate_3 {
   class=MaterialExpressionLinearInterpolate
   inputs [
    A {from=MaterialExpressionMultiply_20}
    B {from=MaterialExpressionLinearInterpolate_2}
    Alpha {from=MaterialExpressionScalarParameter_60}
   ]
  }
  MaterialExpressionLinearInterpolate_4 {
   class=MaterialExpressionLinearInterpolate
   inputs [
    A {from=MaterialExpressionConstant_12}
    B {from=""}
    Alpha {from=MaterialExpressionScalarParameter_65}
   ]
   properties {constB=0.7}
  }
  MaterialExpressionLinearInterpolate_5 {
   class=MaterialExpressionLinearInterpolate
   inputs [
    A {from=MaterialExpressionConstant_9}
    B {from=""}
    Alpha {from=MaterialExpressionScalarParameter_66}
   ]
   properties {constB=0.7}
  }
  MaterialExpressionLinearInterpolate_6 {
   class=MaterialExpressionLinearInterpolate
   inputs [
    A {from=MaterialExpressionMultiply_14}
    B {from=MaterialExpressionIf_1}
    Alpha {from=MaterialExpressionScalarParameter_67}
   ]
  }
  MaterialExpressionMaterialFunctionCall_0 {
   class=MaterialExpressionMaterialFunctionCall
   inputs [
    Overide Texture Set UVs {from=""}
    Override Tech Art Mask UVs {from=""}
    Override Basic Emissive UVs {from=""}
   ]
   properties {materialFunction=&MF_CharacterParent_2023_PreFX}
  }
  MaterialExpressionMaterialFunctionCall_1 {
   class=MaterialExpressionMaterialFunctionCall
   inputs [
    Material {from=MaterialExpressionBlendMaterialAttributes_4}
    M {from=MaterialExpressionMaterialFunctionCall_0, output=M}
    Pre Skinned Local Position {from=MaterialExpressionMaterialFunctionCall_0, output=M}
    Distance Blend {from=MaterialExpressionMaterialFunctionCall_0, output=M}
    Fresnel {from=MaterialExpressionMaterialFunctionCall_0, output=M}
   ]
   properties {materialFunction=&MF_CharacterParent_2023_PostFX}
  }
  MaterialExpressionMaterialFunctionCall_10 {
   class=MaterialExpressionMaterialFunctionCall
   inputs [
    UVs {from=MaterialExpressionMaterialFunctionCall_9, output=Rotated Values}
    Texture Scale {from=MaterialExpressionScalarParameter_32}
   ]
   properties {materialFunction=&ScaleUVsByCenter}
  }
  MaterialExpressionMaterialFunctionCall_11 {
   class=MaterialExpressionMaterialFunctionCall
   inputs [
    UVs {from=MaterialExpressionAdd_24}
    Rotation Center {from=""}
    "Rotation Angle (0-1)" {from=MaterialExpressionScalarParameter_36}
   ]
   properties {materialFunction=&CustomRotator}
  }
  MaterialExpressionMaterialFunctionCall_12 {
   class=MaterialExpressionMaterialFunctionCall
   inputs [
    UVs {from=MaterialExpressionMaterialFunctionCall_11, output=Rotated Values}
    Texture Scale {from=MaterialExpressionMultiply_18}
   ]
   properties {materialFunction=&ScaleUVsByCenter}
  }
  MaterialExpressionMaterialFunctionCall_13 {
   class=MaterialExpressionMaterialFunctionCall
   inputs [
    In {from=MaterialExpressionTextureSampleParameter2D_2, output=R}
    Contrast {from=MaterialExpressionMultiply_15}
   ]
   properties {materialFunction=&CheapContrast}
  }
  MaterialExpressionMaterialFunctionCall_14 {
   class=MaterialExpressionMaterialFunctionCall
   inputs [
    In {from=MaterialExpressionMultiply_5}
    Contrast {from=MaterialExpressionMultiply_16}
   ]
   properties {materialFunction=&CheapContrast}
  }
  MaterialExpressionMaterialFunctionCall_15 {
   class=MaterialExpressionMaterialFunctionCall
   inputs [
    UVs {from=MaterialExpressionAdd_27}
    Rotation Center {from=""}
    "Rotation Angle (0-1)" {from=MaterialExpressionScalarParameter_51}
   ]
   properties {materialFunction=&CustomRotator}
  }
  MaterialExpressionMaterialFunctionCall_16 {
   class=MaterialExpressionMaterialFunctionCall
   inputs [
    UVs {from=MaterialExpressionMaterialFunctionCall_15, output=Rotated Values}
    Texture Scale {from=MaterialExpressionMultiply_19}
   ]
   properties {materialFunction=&ScaleUVsByCenter}
  }
  MaterialExpressionMaterialFunctionCall_17 {
   class=MaterialExpressionMaterialFunctionCall
   inputs [
    Hue Shift Percentage {from=MaterialExpressionScalarParameter_59}
    Texture {from=MaterialExpressionDesaturation_0}
   ]
   properties {materialFunction=&HueShift}
  }
  MaterialExpressionMaterialFunctionCall_2 {
   class=MaterialExpressionMaterialFunctionCall
   inputs [Float3 {from=MaterialExpressionLinearInterpolate_0}]
   properties {materialFunction=&BreakOutFloat3Components}
  }
  MaterialExpressionMaterialFunctionCall_3 {
   class=MaterialExpressionMaterialFunctionCall
   inputs [
    UVs {from=MaterialExpressionAdd_19}
    Rotation Center {from=""}
    "Rotation Angle (0-1)" {from=MaterialExpressionScalarParameter_19}
   ]
   properties {materialFunction=&CustomRotator}
  }
  MaterialExpressionMaterialFunctionCall_4 {
   class=MaterialExpressionMaterialFunctionCall
   inputs [
    UVs {from=MaterialExpressionMaterialFunctionCall_3, output=Rotated Values}
    Texture Scale {from=MaterialExpressionMultiply_17}
   ]
   properties {materialFunction=&ScaleUVsByCenter}
  }
  MaterialExpressionMaterialFunctionCall_5 {
   class=MaterialExpressionMaterialFunctionCall
   inputs [
    UVs {from=MaterialExpressionAdd_20}
    Rotation Center {from=""}
    "Rotation Angle (0-1)" {from=MaterialExpressionScalarParameter_23}
   ]
   properties {materialFunction=&CustomRotator}
  }
  MaterialExpressionMaterialFunctionCall_6 {
   class=MaterialExpressionMaterialFunctionCall
   inputs [
    UVs {from=MaterialExpressionMaterialFunctionCall_5, output=Rotated Values}
    Texture Scale {from=MaterialExpressionScalarParameter_24}
   ]
   properties {materialFunction=&ScaleUVsByCenter}
  }
  MaterialExpressionMaterialFunctionCall_7 {
   class=MaterialExpressionMaterialFunctionCall
   inputs [
    UVs {from=MaterialExpressionAdd_21}
    Rotation Center {from=""}
    "Rotation Angle (0-1)" {from=MaterialExpressionScalarParameter_27}
   ]
   properties {materialFunction=&CustomRotator}
  }
  MaterialExpressionMaterialFunctionCall_8 {
   class=MaterialExpressionMaterialFunctionCall
   inputs [
    UVs {from=MaterialExpressionMaterialFunctionCall_7, output=Rotated Values}
    Texture Scale {from=MaterialExpressionScalarParameter_28}
   ]
   properties {materialFunction=&ScaleUVsByCenter}
  }
  MaterialExpressionMaterialFunctionCall_9 {
   class=MaterialExpressionMaterialFunctionCall
   inputs [
    UVs {from=MaterialExpressionAdd_22}
    Rotation Center {from=""}
    "Rotation Angle (0-1)" {from=MaterialExpressionScalarParameter_31}
   ]
   properties {materialFunction=&CustomRotator}
  }
  MaterialExpressionMax_0 {
   class=MaterialExpressionMax
   inputs [
    A {from=MaterialExpressionTextureSampleParameter2D_0, output=RGB}
    B {from=MaterialExpressionTextureSampleParameter2D_4, output=RGB}
   ]
  }
  MaterialExpressionMax_1 {
   class=MaterialExpressionMax
   inputs [
    A {from=MaterialExpressionTextureSampleParameter2D_5, output=RGB}
    B {from=MaterialExpressionTextureSampleParameter2D_6, output=RGB}
   ]
  }
  MaterialExpressionMax_2 {
   class=MaterialExpressionMax
   inputs [
    A {from=MaterialExpressionMax_0}
    B {from=MaterialExpressionMax_1}
   ]
  }
  MaterialExpressionMultiply_0 {
   class=MaterialExpressionMultiply
   inputs [
    A {from=MaterialExpressionGetMaterialAttributes_0, output=Base Color}
    B {from=MaterialExpressionMultiply_6}
   ]
  }
  MaterialExpressionMultiply_1 {
   class=MaterialExpressionMultiply
   inputs [
    A {from=MaterialExpressionGetMaterialAttributes_1, output=Base Color}
    B {from=MaterialExpressionVectorParameter_1, output=RGB}
   ]
  }
  MaterialExpressionMultiply_10 {
   class=MaterialExpressionMultiply
   inputs [
    A {from=MaterialExpressionDivide_0}
    B {from=MaterialExpressionLinearInterpolate_1}
   ]
  }
  MaterialExpressionMultiply_11 {
   class=MaterialExpressionMultiply
   inputs [
    A {from=MaterialExpressionTextureCoordinate_5}
    B {from=MaterialExpressionConstant_6}
   ]
   properties {constB=5}
  }
  MaterialExpressionMultiply_12 {
   class=MaterialExpressionMultiply
   inputs [
    A {from=MaterialExpressionPower_0}
    B {from=MaterialExpressionPower_1}
   ]
  }
  MaterialExpressionMultiply_13 {
   class=MaterialExpressionMultiply
   inputs [
    A {from=MaterialExpressionOneMinus_1}
    B {from=""}
   ]
   properties {constB=2}
  }
  MaterialExpressionMultiply_14 {
   class=MaterialExpressionMultiply
   inputs [
    A {from=MaterialExpressionMultiply_7}
    B {from=MaterialExpressionIf_0}
   ]
  }
  MaterialExpressionMultiply_15 {
   class=MaterialExpressionMultiply
   inputs [
    A {from=MaterialExpressionScalarParameter_41}
    B {from=MaterialExpressionScalarParameter_43}
   ]
  }
  MaterialExpressionMultiply_16 {
   class=MaterialExpressionMultiply
   inputs [
    A {from=MaterialExpressionScalarParameter_42}
    B {from=MaterialExpressionScalarParameter_44}
   ]
  }
  MaterialExpressionMultiply_17 {
   class=MaterialExpressionMultiply
   inputs [
    A {from=MaterialExpressionAppendVector_9}
    B {from=MaterialExpressionScalarParameter_46}
   ]
   properties {constB=1.2}
  }
  MaterialExpressionMultiply_18 {
   class=MaterialExpressionMultiply
   inputs [
    A {from=MaterialExpressionAppendVector_10}
    B {from=MaterialExpressionScalarParameter_37}
   ]
  }
  MaterialExpressionMultiply_19 {
   class=MaterialExpressionMultiply
   inputs [
    A {from=MaterialExpressionAppendVector_12}
    B {from=MaterialExpressionScalarParameter_52}
   ]
  }
  MaterialExpressionMultiply_2 {
   class=MaterialExpressionMultiply
   inputs [
    A {from=MaterialExpressionGetMaterialAttributes_2, output=Base Color}
    B {from=MaterialExpressionVectorParameter_2, output=RGB}
   ]
  }
  MaterialExpressionMultiply_20 {
   class=MaterialExpressionMultiply
   inputs [
    A {from=MaterialExpressionMaterialFunctionCall_17, output=Result}
    B {from=MaterialExpressionVectorParameter_5, output=RGB}
   ]
  }
  MaterialExpressionMultiply_21 {
   class=MaterialExpressionMultiply
   inputs [
    A {from=MaterialExpressionMultiply_20}
    B {from=MaterialExpressionScalarParameter_63}
   ]
   properties {constB=2}
  }
  MaterialExpressionMultiply_3 {
   class=MaterialExpressionMultiply
   inputs [
    A {from=MaterialExpressionGetMaterialAttributes_3, output=Base Color}
    B {from=MaterialExpressionVectorParameter_3, output=RGB}
   ]
  }
  MaterialExpressionMultiply_4 {
   class=MaterialExpressionMultiply
   inputs [
    A {from=MaterialExpressionGetMaterialAttributes_4, output=Base Color}
    B {from=MaterialExpressionLinearInterpolate_6}
   ]
  }
  MaterialExpressionMultiply_5 {
   class=MaterialExpressionMultiply
   inputs [
    A {from=MaterialExpressionTextureSampleParameter2D_3, output=R}
    B {from=MaterialExpressionSwitch_1}
   ]
  }
  MaterialExpressionMultiply_6 {
   class=MaterialExpressionMultiply
   inputs [
    A {from=MaterialExpressionVectorParameter_0, output=RGB}
    B {from=MaterialExpressionScalarParameter_38}
   ]
  }
  MaterialExpressionMultiply_7 {
   class=MaterialExpressionMultiply
   inputs [
    A {from=MaterialExpressionVectorParameter_4, output=RGB}
    B {from=MaterialExpressionScalarParameter_39}
   ]
  }
  MaterialExpressionMultiply_8 {
   class=MaterialExpressionMultiply
   inputs [
    A {from=MaterialExpressionOneMinus_0}
    B {from=""}
   ]
   properties {constB=2}
  }
  MaterialExpressionMultiply_9 {
   class=MaterialExpressionMultiply
   inputs [
    A {from=MaterialExpressionMultiply_10}
    B {from=""}
   ]
   properties {constB=-1}
  }
  MaterialExpressionNamedRerouteDeclaration_0 {
   class=MaterialExpressionNamedRerouteDeclaration
   inputs [None {from=MaterialExpressionMaterialFunctionCall_0, output=Result}]
   properties {name=SourceMA, nodeColor=<rgba 0.3594 0 1 1>}
  }
  MaterialExpressionNamedRerouteDeclaration_1 {
   class=MaterialExpressionNamedRerouteDeclaration
   inputs [None {from=MaterialExpressionSetMaterialAttributes_0}]
   properties {name=MetalBase, nodeColor=<rgba 0 1 0.0391 1>}
  }
  MaterialExpressionNamedRerouteDeclaration_2 {
   class=MaterialExpressionNamedRerouteDeclaration
   inputs [None {from=MaterialExpressionSetMaterialAttributes_1}]
   properties {name=MetalAccent, nodeColor=<rgba 0 1 0.0391 1>}
  }
  MaterialExpressionNamedRerouteDeclaration_3 {
   class=MaterialExpressionNamedRerouteDeclaration
   inputs [None {from=MaterialExpressionSetMaterialAttributes_2}]
   properties {name=Cloth, nodeColor=<rgba 0 1 0.0391 1>}
  }
  MaterialExpressionNamedRerouteDeclaration_4 {
   class=MaterialExpressionNamedRerouteDeclaration
   inputs [None {from=MaterialExpressionSetMaterialAttributes_3}]
   properties {name=Decal, nodeColor=<rgba 0 1 0.0391 1>}
  }
  MaterialExpressionNamedRerouteDeclaration_5 {
   class=MaterialExpressionNamedRerouteDeclaration
   inputs [None {from=MaterialExpressionSetMaterialAttributes_4}]
   properties {name=Sigil, nodeColor=<rgba 0 1 0.0391 1>}
  }
  MaterialExpressionNamedRerouteDeclaration_6 {
   class=MaterialExpressionNamedRerouteDeclaration
   inputs [None {from=MaterialExpressionMax_2}]
   properties {name=Mask edge, nodeColor=<rgba 1 0.6797 0 1>}
  }
  MaterialExpressionNamedRerouteDeclaration_7 {
   class=MaterialExpressionNamedRerouteDeclaration
   inputs [None {from=MaterialExpressionSetMaterialAttributes_5}]
   properties {name=Leather, nodeColor=<rgba 0 1 0.0391 1>}
  }
  MaterialExpressionNamedRerouteUsage_0 {class=MaterialExpressionNamedRerouteUsage}
  MaterialExpressionNamedRerouteUsage_1 {class=MaterialExpressionNamedRerouteUsage}
  MaterialExpressionNamedRerouteUsage_10 {class=MaterialExpressionNamedRerouteUsage}
  MaterialExpressionNamedRerouteUsage_11 {class=MaterialExpressionNamedRerouteUsage}
  MaterialExpressionNamedRerouteUsage_12 {class=MaterialExpressionNamedRerouteUsage}
  MaterialExpressionNamedRerouteUsage_14 {class=MaterialExpressionNamedRerouteUsage}
  MaterialExpressionNamedRerouteUsage_2 {class=MaterialExpressionNamedRerouteUsage}
  MaterialExpressionNamedRerouteUsage_3 {class=MaterialExpressionNamedRerouteUsage}
  MaterialExpressionNamedRerouteUsage_4 {class=MaterialExpressionNamedRerouteUsage}
  MaterialExpressionNamedRerouteUsage_5 {class=MaterialExpressionNamedRerouteUsage}
  MaterialExpressionNamedRerouteUsage_6 {class=MaterialExpressionNamedRerouteUsage}
  MaterialExpressionNamedRerouteUsage_7 {class=MaterialExpressionNamedRerouteUsage}
  MaterialExpressionNamedRerouteUsage_8 {class=MaterialExpressionNamedRerouteUsage}
  MaterialExpressionNamedRerouteUsage_9 {class=MaterialExpressionNamedRerouteUsage}
  MaterialExpressionOneMinus_0 {class=MaterialExpressionOneMinus, inputs [None {from=MaterialExpressionTextureSampleParameter2D_7, output=R}]}
  MaterialExpressionOneMinus_1 {class=MaterialExpressionOneMinus, inputs [None {from=MaterialExpressionTextureSampleParameter2D_8, output=R}]}
  MaterialExpressionPower_0 {
   class=MaterialExpressionPower
   inputs [
    Base {from=MaterialExpressionMultiply_8}
    Exp {from=""}
   ]
   properties {constExponent=10}
  }
  MaterialExpressionPower_1 {
   class=MaterialExpressionPower
   inputs [
    Base {from=MaterialExpressionMultiply_13}
    Exp {from=""}
   ]
   properties {constExponent=6}
  }
  MaterialExpressionPower_2 {
   class=MaterialExpressionPower
   inputs [
    Base {from=MaterialExpressionFresnel_0}
    Exp {from=MaterialExpressionScalarParameter_64}
   ]
  }
  MaterialExpressionReroute_0 {class=MaterialExpressionReroute, inputs [None {from=MaterialExpressionSaturate_5}]}
  MaterialExpressionSaturate_0 {class=MaterialExpressionSaturate, inputs [None {from=MaterialExpressionAdd_12}]}
  MaterialExpressionSaturate_1 {class=MaterialExpressionSaturate, inputs [None {from=MaterialExpressionAdd_9}]}
  MaterialExpressionSaturate_2 {class=MaterialExpressionSaturate, inputs [None {from=MaterialExpressionPower_0}]}
  MaterialExpressionSaturate_3 {class=MaterialExpressionSaturate, inputs [None {from=MaterialExpressionPower_2}]}
  MaterialExpressionSaturate_4 {class=MaterialExpressionSaturate, inputs [None {from=MaterialExpressionMaterialFunctionCall_13, output=Result}]}
  MaterialExpressionSaturate_5 {class=MaterialExpressionSaturate, inputs [None {from=MaterialExpressionMaterialFunctionCall_14, output=Result}]}
  MaterialExpressionScalarParameter_0 {class=MaterialExpressionScalarParameter, properties {group=WingBath, parameterName=MetalBase_RoughnessBias, sliderMax=1, sliderMin=-1}}
  MaterialExpressionScalarParameter_1 {class=MaterialExpressionScalarParameter, properties {group=WingBath, parameterName=MetalBase_SpecularBias, sliderMax=1, sliderMin=-1}}
  MaterialExpressionScalarParameter_10 {class=MaterialExpressionScalarParameter, properties {group=WingBath, parameterName=Decal_SpecularBias, sliderMax=1, sliderMin=-1}}
  MaterialExpressionScalarParameter_11 {class=MaterialExpressionScalarParameter, properties {group=WingBath, parameterName=Decal_MetallicBias, sliderMax=1, sliderMin=-1}}
  MaterialExpressionScalarParameter_12 {
   class=MaterialExpressionScalarParameter
   properties {defaultValue=0.2, group=WingBath, parameterName=Sigil_RoughnessBias, sliderMax=1, sliderMin=-1}
  }
  MaterialExpressionScalarParameter_13 {class=MaterialExpressionScalarParameter, properties {group=WingBath, parameterName=Sigil_SpecularBias, sliderMax=1, sliderMin=-1}}
  MaterialExpressionScalarParameter_14 {class=MaterialExpressionScalarParameter, properties {group=WingBath, parameterName=Sigil_MetallicBias, sliderMax=1, sliderMin=-1}}
  MaterialExpressionScalarParameter_15 {class=MaterialExpressionScalarParameter, properties {defaultValue=1, group=WingBath, parameterName=UV_Edge_bias_MIn}}
  MaterialExpressionScalarParameter_16 {class=MaterialExpressionScalarParameter, properties {defaultValue=-0.32, group=WingBath, parameterName=SigilU}}
  MaterialExpressionScalarParameter_17 {class=MaterialExpressionScalarParameter, properties {defaultValue=2, group=WingBath, parameterName=UV_Edge_bias_Max}}
  MaterialExpressionScalarParameter_18 {class=MaterialExpressionScalarParameter, properties {defaultValue=0.25, group=WingBath, parameterName=SigilV}}
  MaterialExpressionScalarParameter_19 {class=MaterialExpressionScalarParameter, properties {defaultValue=0.75, group=WingBath, parameterName=SigilRoation}}
  MaterialExpressionScalarParameter_2 {class=MaterialExpressionScalarParameter, properties {group=WingBath, parameterName=MetalBase_MetallicBias, sliderMax=1, sliderMin=-1}}
  MaterialExpressionScalarParameter_20 {class=MaterialExpressionScalarParameter, properties {defaultValue=0.26, group=WingBath, parameterName=sigilScale}}
  MaterialExpressionScalarParameter_21 {class=MaterialExpressionScalarParameter, properties {defaultValue=-0.336, group=WingBath, parameterName=SigilU2}}
  MaterialExpressionScalarParameter_22 {class=MaterialExpressionScalarParameter, properties {defaultValue=-0.25, group=WingBath, parameterName=SigilV2}}
  MaterialExpressionScalarParameter_23 {class=MaterialExpressionScalarParameter, properties {defaultValue=0.75, group=WingBath, parameterName=SigilRoation2}}
  MaterialExpressionScalarParameter_24 {class=MaterialExpressionScalarParameter, properties {defaultValue=0.24, group=WingBath, parameterName=sigilScale2}}
  MaterialExpressionScalarParameter_25 {class=MaterialExpressionScalarParameter, properties {defaultValue=0.312, group=WingBath, parameterName=SigilU3}}
  MaterialExpressionScalarParameter_26 {class=MaterialExpressionScalarParameter, properties {defaultValue=-0.25, group=WingBath, parameterName=SigilV3}}
  MaterialExpressionScalarParameter_27 {class=MaterialExpressionScalarParameter, properties {defaultValue=0.756, group=WingBath, parameterName=SigilRoation3}}
  MaterialExpressionScalarParameter_28 {class=MaterialExpressionScalarParameter, properties {defaultValue=0.2, group=WingBath, parameterName=sigilScale3}}
  MaterialExpressionScalarParameter_29 {class=MaterialExpressionScalarParameter, properties {defaultValue=0.322, group=WingBath, parameterName=SigilU4}}
  MaterialExpressionScalarParameter_3 {class=MaterialExpressionScalarParameter, properties {group=WingBath, parameterName=MetalAccent_RoughnessBias, sliderMax=1, sliderMin=-1}}
  MaterialExpressionScalarParameter_30 {class=MaterialExpressionScalarParameter, properties {defaultValue=0.258, group=WingBath, parameterName=SigilV4}}
  MaterialExpressionScalarParameter_31 {class=MaterialExpressionScalarParameter, properties {defaultValue=0.25, group=WingBath, parameterName=SigilRoation4}}
  MaterialExpressionScalarParameter_32 {class=MaterialExpressionScalarParameter, properties {defaultValue=0.2, group=WingBath, parameterName=sigilScale4}}
  MaterialExpressionScalarParameter_33 {class=MaterialExpressionScalarParameter, properties {group=WingBath, parameterName=Sigil_Style_Switch}}
  MaterialExpressionScalarParameter_34 {class=MaterialExpressionScalarParameter, properties {defaultValue=-0.2465, group=WingBath, parameterName=SigilU5}}
  MaterialExpressionScalarParameter_35 {class=MaterialExpressionScalarParameter, properties {defaultValue=0.055, group=WingBath, parameterName=SigilV5}}
  MaterialExpressionScalarParameter_36 {class=MaterialExpressionScalarParameter, properties {defaultValue=-0.005, group=WingBath, parameterName=SigilRoation5}}
  MaterialExpressionScalarParameter_37 {class=MaterialExpressionScalarParameter, properties {defaultValue=0.185, group=WingBath, parameterName=sigilScale5}}
  MaterialExpressionScalarParameter_38 {class=MaterialExpressionScalarParameter, properties {defaultValue=1, group=WingBath, parameterName=MetalBase_darker}}
  MaterialExpressionScalarParameter_39 {class=MaterialExpressionScalarParameter, properties {defaultValue=1, group=WingBath, parameterName=Sigil_Brightness}}
  MaterialExpressionScalarParameter_4 {class=MaterialExpressionScalarParameter, properties {group=WingBath, parameterName=MetalAccent_SpecularBias, sliderMax=1, sliderMin=-1}}
  MaterialExpressionScalarParameter_40 {class=MaterialExpressionScalarParameter, properties {defaultValue=1, group=WingBath, parameterName=Sigil_Brightness_Mul}}
  MaterialExpressionScalarParameter_41 {class=MaterialExpressionScalarParameter, properties {group=WingBath, parameterName=Cheap_Contrast}}
  MaterialExpressionScalarParameter_42 {class=MaterialExpressionScalarParameter, properties {group=WingBath, parameterName=Cheap_Contrast}}
  MaterialExpressionScalarParameter_43 {class=MaterialExpressionScalarParameter, properties {group=WingBath, parameterName=Cheap_Contrast_Switch}}
  MaterialExpressionScalarParameter_44 {class=MaterialExpressionScalarParameter, properties {group=WingBath, parameterName=Cheap_Contrast_Switch}}
  MaterialExpressionScalarParameter_45 {class=MaterialExpressionScalarParameter, properties {defaultValue=0.228, group=WingBath, parameterName=sigilScaleY}}
  MaterialExpressionScalarParameter_46 {class=MaterialExpressionScalarParameter, properties {defaultValue=1.03, parameterName=Scale_bias}}
  MaterialExpressionScalarParameter_47 {class=MaterialExpressionScalarParameter, properties {defaultValue=1.15, group=WingBath, parameterName=sigilScale_capeA}}
  MaterialExpressionScalarParameter_48 {class=MaterialExpressionScalarParameter, properties {defaultValue=0.9, group=WingBath, parameterName=sigilScale_capeB}}
  MaterialExpressionScalarParameter_49 {class=MaterialExpressionScalarParameter, properties {defaultValue=-0.245, group=WingBath, parameterName=SigilU6}}
  MaterialExpressionScalarParameter_5 {class=MaterialExpressionScalarParameter, properties {group=WingBath, parameterName=MetalAccent_MetallicBias, sliderMax=1, sliderMin=-1}}
  MaterialExpressionScalarParameter_50 {class=MaterialExpressionScalarParameter, properties {defaultValue=0.1125, group=WingBath, parameterName=SigilV6}}
  MaterialExpressionScalarParameter_51 {class=MaterialExpressionScalarParameter, properties {group=WingBath, parameterName=SigilRoation6}}
  MaterialExpressionScalarParameter_52 {class=MaterialExpressionScalarParameter, properties {defaultValue=0.24, group=WingBath, parameterName=sigilScale6}}
  MaterialExpressionScalarParameter_53 {class=MaterialExpressionScalarParameter, properties {defaultValue=0.848, group=WingBath, parameterName=sigilScale_cape_Universal_X}}
  MaterialExpressionScalarParameter_54 {class=MaterialExpressionScalarParameter, properties {defaultValue=1, group=WingBath, parameterName=sigilScale_cape_Universal_Y}}
  MaterialExpressionScalarParameter_55 {class=MaterialExpressionScalarParameter, properties {group=WingBath, parameterName=Leather_RoughnessBias, sliderMax=1, sliderMin=-1}}
  MaterialExpressionScalarParameter_56 {class=MaterialExpressionScalarParameter, properties {group=WingBath, parameterName=Leather_SpecularBias, sliderMax=1, sliderMin=-1}}
  MaterialExpressionScalarParameter_57 {class=MaterialExpressionScalarParameter, properties {group=WingBath, parameterName=Leather_MetallicBias, sliderMax=1, sliderMin=-1}}
  MaterialExpressionScalarParameter_58 {class=MaterialExpressionScalarParameter, properties {group=WingBath, parameterName=Leather_Desaturation}}
  MaterialExpressionScalarParameter_59 {class=MaterialExpressionScalarParameter, properties {group=WingBath, parameterName=Leather_Hueshift}}
  MaterialExpressionScalarParameter_6 {class=MaterialExpressionScalarParameter, properties {group=WingBath, parameterName=Cloth_RoughnessBias, sliderMax=1, sliderMin=-1}}
  MaterialExpressionScalarParameter_60 {class=MaterialExpressionScalarParameter, properties {group=WingBath, parameterName=Leather_UseFrenel}}
  MaterialExpressionScalarParameter_61 {class=MaterialExpressionScalarParameter, properties {defaultValue=5, group=WingBath, parameterName=Leather_Exponentln}}
  MaterialExpressionScalarParameter_62 {class=MaterialExpressionScalarParameter, properties {group=WingBath, parameterName=Leather_BaseReflectFractiinln}}
  MaterialExpressionScalarParameter_63 {class=MaterialExpressionScalarParameter, properties {group=WingBath, parameterName=Leather_Mul}}
  MaterialExpressionScalarParameter_64 {class=MaterialExpressionScalarParameter, properties {defaultValue=1, group=WingBath, parameterName=Leather_power}}
  MaterialExpressionScalarParameter_65 {class=MaterialExpressionScalarParameter, properties {group=WingBath, parameterName=Cape?}}
  MaterialExpressionScalarParameter_66 {class=MaterialExpressionScalarParameter, properties {group=WingBath, parameterName=Cape?}}
  MaterialExpressionScalarParameter_67 {class=MaterialExpressionScalarParameter, properties {group=WingBath, parameterName=Cape?}}
  MaterialExpressionScalarParameter_7 {class=MaterialExpressionScalarParameter, properties {group=WingBath, parameterName=Cloth_SpecularBias, sliderMax=1, sliderMin=-1}}
  MaterialExpressionScalarParameter_8 {class=MaterialExpressionScalarParameter, properties {group=WingBath, parameterName=Cloth_MetallicBias, sliderMax=1, sliderMin=-1}}
  MaterialExpressionScalarParameter_9 {class=MaterialExpressionScalarParameter, properties {group=WingBath, parameterName=Decal_RoughnessBias, sliderMax=1, sliderMin=-1}}
  MaterialExpressionSetMaterialAttributes_0 {
   class=MaterialExpressionSetMaterialAttributes
   inputs [
    MaterialAttributes {from=MaterialExpressionGetMaterialAttributes_0, output=MaterialAttributes}
    Base Color {from=MaterialExpressionMultiply_0, output=""}
    Metallic {from=MaterialExpressionAdd_0, output=""}
    Specular {from=MaterialExpressionAdd_1, output=""}
    Roughness {from=MaterialExpressionAdd_2, output=""}
   ]
   properties {attributeSetTypes [
     #1
     #2
     #3
     #4
    ]}
  }
  MaterialExpressionSetMaterialAttributes_1 {
   class=MaterialExpressionSetMaterialAttributes
   inputs [
    MaterialAttributes {from=MaterialExpressionGetMaterialAttributes_1, output=MaterialAttributes}
    Base Color {from=MaterialExpressionMultiply_1, output=""}
    Metallic {from=MaterialExpressionAdd_3, output=""}
    Specular {from=MaterialExpressionAdd_4, output=""}
    Roughness {from=MaterialExpressionAdd_5, output=""}
   ]
   properties {attributeSetTypes [
     #1
     #2
     #3
     #4
    ]}
  }
  MaterialExpressionSetMaterialAttributes_2 {
   class=MaterialExpressionSetMaterialAttributes
   inputs [
    MaterialAttributes {from=MaterialExpressionGetMaterialAttributes_2, output=MaterialAttributes}
    Base Color {from=MaterialExpressionMultiply_2, output=""}
    Metallic {from=MaterialExpressionAdd_6, output=""}
    Specular {from=MaterialExpressionAdd_7, output=""}
    Roughness {from=MaterialExpressionAdd_8, output=""}
   ]
   properties {attributeSetTypes [
     #1
     #2
     #3
     #4
    ]}
  }
  MaterialExpressionSetMaterialAttributes_3 {
   class=MaterialExpressionSetMaterialAttributes
   inputs [
    MaterialAttributes {from=MaterialExpressionGetMaterialAttributes_3, output=MaterialAttributes}
    Base Color {from=MaterialExpressionMultiply_3, output=""}
    Metallic {from=MaterialExpressionSaturate_1, output=""}
    Specular {from=MaterialExpressionAdd_10, output=""}
    Roughness {from=MaterialExpressionLinearInterpolate_4, output=""}
   ]
   properties {attributeSetTypes [
     #1
     #2
     #3
     #4
    ]}
  }
  MaterialExpressionSetMaterialAttributes_4 {
   class=MaterialExpressionSetMaterialAttributes
   inputs [
    MaterialAttributes {from=MaterialExpressionGetMaterialAttributes_4, output=MaterialAttributes}
    Base Color {from=MaterialExpressionMultiply_4, output=""}
    Metallic {from=MaterialExpressionConstant_5, output=""}
    Specular {from=MaterialExpressionAdd_13, output=""}
    Roughness {from=MaterialExpressionLinearInterpolate_5, output=""}
   ]
   properties {attributeSetTypes [
     #1
     #2
     #3
     #4
    ]}
  }
  MaterialExpressionSetMaterialAttributes_5 {
   class=MaterialExpressionSetMaterialAttributes
   inputs [
    MaterialAttributes {from=MaterialExpressionGetMaterialAttributes_5, output=MaterialAttributes}
    Base Color {from=MaterialExpressionLinearInterpolate_3, output=""}
    Metallic {from=MaterialExpressionAdd_28, output=""}
    Specular {from=MaterialExpressionAdd_29, output=""}
    Roughness {from=MaterialExpressionAdd_30, output=""}
   ]
   properties {attributeSetTypes [
     #1
     #2
     #3
     #4
    ]}
  }
  MaterialExpressionSubtract_0 {
   class=MaterialExpressionSubtract
   inputs [
    A {from=MaterialExpressionVectorParameter_6, output=R}
    B {from=MaterialExpressionVectorParameter_4, output=R}
   ]
  }
  MaterialExpressionSwitch_0 {
   class=MaterialExpressionSwitch
   inputs [
    SwitchValue {from=MaterialExpressionScalarParameter_33, output=""}
    Default {from="", output=""}
    None {from=MaterialExpressionMaterialFunctionCall_4, output=UVs}
    None {from=MaterialExpressionMaterialFunctionCall_6, output=UVs}
    None {from=MaterialExpressionMaterialFunctionCall_8, output=UVs}
    None {from=MaterialExpressionMaterialFunctionCall_10, output=UVs}
    None {from=MaterialExpressionMaterialFunctionCall_12, output=UVs}
    None {from=MaterialExpressionMaterialFunctionCall_16, output=UVs}
   ]
   properties {inputs [
     {
      inputName=None
      input {
       expression=&self:MaterialExpressionMaterialFunctionCall_4
       inputName=None
       mask=0
       maskA=0
       maskB=0
       maskG=0
       maskR=0
       outputIndex=0
      }
     }
     {
      inputName=None
      input {
       expression=&self:MaterialExpressionMaterialFunctionCall_6
       inputName=None
       mask=0
       maskA=0
       maskB=0
       maskG=0
       maskR=0
       outputIndex=0
      }
     }
     {
      inputName=None
      input {
       expression=&self:MaterialExpressionMaterialFunctionCall_8
       inputName=None
       mask=0
       maskA=0
       maskB=0
       maskG=0
       maskR=0
       outputIndex=0
      }
     }
     {
      inputName=None
      input {
       expression=&self:MaterialExpressionMaterialFunctionCall_10
       inputName=None
       mask=0
       maskA=0
       maskB=0
       maskG=0
       maskR=0
       outputIndex=0
      }
     }
     {
      inputName=None
      input {
       expression=&self:MaterialExpressionMaterialFunctionCall_12
       inputName=None
       mask=0
       maskA=0
       maskB=0
       maskG=0
       maskR=0
       outputIndex=0
      }
     }
     {
      inputName=None
      input {
       expression=&self:MaterialExpressionMaterialFunctionCall_16
       inputName=None
       mask=0
       maskA=0
       maskB=0
       maskG=0
       maskR=0
       outputIndex=0
      }
     }
    ]}
  }
  MaterialExpressionSwitch_1 {
   class=MaterialExpressionSwitch
   inputs [
    SwitchValue {from=MaterialExpressionScalarParameter_33, output=""}
    Default {from="", output=""}
    None {from=MaterialExpressionConvert_0, output=""}
    None {from=MaterialExpressionConvert_0, output=""}
    None {from=MaterialExpressionConvert_0, output=""}
    None {from=MaterialExpressionConvert_0, output=""}
    None {from=MaterialExpressionSaturate_2, output=""}
    None {from=MaterialExpressionSaturate_2, output=""}
   ]
   properties {inputs [
     {
      inputName=None
      input {
       expression=&self:MaterialExpressionConvert_0
       inputName=None
       mask=0
       maskA=0
       maskB=0
       maskG=0
       maskR=0
       outputIndex=0
      }
     }
     {
      inputName=None
      input {
       expression=&self:MaterialExpressionConvert_0
       inputName=None
       mask=0
       maskA=0
       maskB=0
       maskG=0
       maskR=0
       outputIndex=1
      }
     }
     {
      inputName=None
      input {
       expression=&self:MaterialExpressionConvert_0
       inputName=None
       mask=0
       maskA=0
       maskB=0
       maskG=0
       maskR=0
       outputIndex=0
      }
     }
     {
      inputName=None
      input {
       expression=&self:MaterialExpressionConvert_0
       inputName=None
       mask=0
       maskA=0
       maskB=0
       maskG=0
       maskR=0
       outputIndex=1
      }
     }
     {
      inputName=None
      input {
       expression=&self:MaterialExpressionSaturate_2
       inputName=None
       mask=0
       maskA=0
       maskB=0
       maskG=0
       maskR=0
       outputIndex=0
      }
     }
     {
      inputName=None
      input {
       expression=&self:MaterialExpressionSaturate_2
       inputName=None
       mask=0
       maskA=0
       maskB=0
       maskG=0
       maskR=0
       outputIndex=0
      }
     }
    ]}
  }
  MaterialExpressionTextureCoordinate_0 {class=MaterialExpressionTextureCoordinate}
  MaterialExpressionTextureCoordinate_1 {class=MaterialExpressionTextureCoordinate}
  MaterialExpressionTextureCoordinate_10 {class=MaterialExpressionTextureCoordinate}
  MaterialExpressionTextureCoordinate_2 {class=MaterialExpressionTextureCoordinate}
  MaterialExpressionTextureCoordinate_3 {class=MaterialExpressionTextureCoordinate}
  MaterialExpressionTextureCoordinate_4 {class=MaterialExpressionTextureCoordinate}
  MaterialExpressionTextureCoordinate_5 {class=MaterialExpressionTextureCoordinate}
  MaterialExpressionTextureCoordinate_6 {class=MaterialExpressionTextureCoordinate}
  MaterialExpressionTextureCoordinate_7 {class=MaterialExpressionTextureCoordinate}
  MaterialExpressionTextureCoordinate_8 {class=MaterialExpressionTextureCoordinate}
  MaterialExpressionTextureCoordinate_9 {class=MaterialExpressionTextureCoordinate}
  MaterialExpressionTextureSampleParameter2D_0 {
   class=MaterialExpressionTextureSampleParameter2D
   inputs [
    UVs {from=MaterialExpressionAdd_15, output=""}
    Apply View MipBias {from="", output=""}
   ]
   properties {
    desc="R_MetalBase,G_MetalAccent B_Cloth"
    group=WingBath
    parameterName=TechArtMask_CustomMaterial
    samplerSource=SSM_Wrap_WorldGroupSettings
    samplerType=SAMPLERTYPE_Masks
    texture=&T_BlackMask
   }
  }
  MaterialExpressionTextureSampleParameter2D_1 {
   class=MaterialExpressionTextureSampleParameter2D
   inputs [
    UVs {from="", output=""}
    Apply View MipBias {from="", output=""}
   ]
   properties {
    desc=G_Leather
    group=WingBath
    parameterName=TechArtMask_Leather
    samplerSource=SSM_Clamp_WorldGroupSettings
    samplerType=SAMPLERTYPE_Masks
    texture=&T_BlackMask
   }
  }
  MaterialExpressionTextureSampleParameter2D_2 {
   class=MaterialExpressionTextureSampleParameter2D
   inputs [
    UVs {from="", output=""}
    Apply View MipBias {from="", output=""}
   ]
   properties {
    desc=R_decal
    group=WingBath
    parameterName=TechArtMask_Decal
    samplerType=SAMPLERTYPE_Masks
    texture=&T_BlackMask
   }
  }
  MaterialExpressionTextureSampleParameter2D_3 {
   class=MaterialExpressionTextureSampleParameter2D
   inputs [
    UVs {from=MaterialExpressionSwitch_0, output=""}
    Apply View MipBias {from="", output=""}
   ]
   properties {
    desc=R_Sigil
    group=WingBath
    parameterName=TechArtMask_Sigil
    samplerSource=SSM_Clamp_WorldGroupSettings
    samplerType=SAMPLERTYPE_Masks
    texture=&T_BlackMask
   }
  }
  MaterialExpressionTextureSampleParameter2D_4 {
   class=MaterialExpressionTextureSampleParameter2D
   inputs [
    UVs {from=MaterialExpressionAdd_16, output=""}
    Apply View MipBias {from="", output=""}
   ]
   properties {
    desc="R_MetalBase,G_MetalAccent B_Cloth"
    group=WingBath
    parameterName=TechArtMask_CustomMaterial
    samplerSource=SSM_Wrap_WorldGroupSettings
    samplerType=SAMPLERTYPE_Masks
    texture=&T_BlackMask
   }
  }
  MaterialExpressionTextureSampleParameter2D_5 {
   class=MaterialExpressionTextureSampleParameter2D
   inputs [
    UVs {from=MaterialExpressionAdd_17, output=""}
    Apply View MipBias {from="", output=""}
   ]
   properties {
    desc="R_MetalBase,G_MetalAccent B_Cloth"
    group=WingBath
    parameterName=TechArtMask_CustomMaterial
    samplerSource=SSM_Wrap_WorldGroupSettings
    samplerType=SAMPLERTYPE_Masks
    texture=&T_BlackMask
   }
  }
  MaterialExpressionTextureSampleParameter2D_6 {
   class=MaterialExpressionTextureSampleParameter2D
   inputs [
    UVs {from=MaterialExpressionAdd_18, output=""}
    Apply View MipBias {from="", output=""}
   ]
   properties {
    desc="R_MetalBase,G_MetalAccent B_Cloth"
    group=WingBath
    parameterName=TechArtMask_CustomMaterial
    samplerSource=SSM_Wrap_WorldGroupSettings
    samplerType=SAMPLERTYPE_Masks
    texture=&T_BlackMask
   }
  }
  MaterialExpressionTextureSampleParameter2D_7 {
   class=MaterialExpressionTextureSampleParameter2D
   inputs [
    UVs {from="", output=""}
    Apply View MipBias {from="", output=""}
   ]
   properties {
    desc="R = Spec , G = Roughness, B= Metal"
    group=00_Texture Inputs
    parameterName=SRM
    samplerType=SAMPLERTYPE_Masks
    sortPriority=1
    texture=&T_Fortnite_Default_S
   }
  }
  MaterialExpressionTextureSampleParameter2D_8 {
   class=MaterialExpressionTextureSampleParameter2D
   inputs [
    UVs {from="", output=""}
    Apply View MipBias {from="", output=""}
   ]
   properties {
    group=00_Texture Inputs
    parameterName=Mask
    samplerType=SAMPLERTYPE_Masks
    sortPriority=3
    texture=&T_BlackMask
   }
  }
  MaterialExpressionTextureSample_0 {
   class=MaterialExpressionTextureSample
   inputs [
    UVs {from=MaterialExpressionMultiply_11}
    Tex {from=""}
    Apply View MipBias {from=""}
   ]
   properties {samplerType=SAMPLERTYPE_Grayscale, texture=/Game/Global/Textures/Default/Noise/TilingNoise06}
  }
  MaterialExpressionTextureSample_3 {
   class=MaterialExpressionTextureSample
   inputs [
    UVs {from=""}
    Tex {from=""}
    Apply View MipBias {from=""}
   ]
   properties {texture=/BRCosmetics/Accessories/FORT_Capes/Cape_M_MED_WingBathUniversal/Textures/T_WingBathUniversal_Cape_M}
  }
  MaterialExpressionVectorParameter_0 {class=MaterialExpressionVectorParameter, properties {group=WingBath, parameterName=MetalBase_Color, defaultValue=<rgba 1 1 1 1>}}
  MaterialExpressionVectorParameter_1 {class=MaterialExpressionVectorParameter, properties {group=WingBath, parameterName=MetalAccent_Color, defaultValue=<rgba 1 1 1 1>}}
  MaterialExpressionVectorParameter_2 {class=MaterialExpressionVectorParameter, properties {group=WingBath, parameterName=Cloth_Color, defaultValue=<rgba 1 1 1 1>}}
  MaterialExpressionVectorParameter_3 {class=MaterialExpressionVectorParameter, properties {group=WingBath, parameterName=Decal_Color, defaultValue=<rgba 1 1 1 1>}}
  MaterialExpressionVectorParameter_4 {class=MaterialExpressionVectorParameter, properties {group=WingBath, parameterName=Sigil_Color, defaultValue=<rgba 1 1 1 1>}}
  MaterialExpressionVectorParameter_5 {class=MaterialExpressionVectorParameter, properties {group=WingBath, parameterName=Leather_Color, defaultValue=<rgba 1 1 1 1>}}
  MaterialExpressionVectorParameter_6 {class=MaterialExpressionVectorParameter, properties {group=WingBath, parameterName=Decal_Color, defaultValue=<rgba 1 1 1 1>}}
 }
 properties {
  bUseMaterialAttributes=true
  bUsedWithClothing=true
  bUsedWithMeshDeformer=true
  bUsedWithMorphTargets=true
  bUsedWithSkeletalMesh=true
  pixelDepthOffsetMode=PDOM_Legacy
  shadingModel=MSM_FromMaterialExpression
  subsurfaceProfile=&SS_HeroSkin_02
  usedShadingModels=Default Lit | Subsurface | Clear Coat | Subsurface Profile | Cloth
 }
}
