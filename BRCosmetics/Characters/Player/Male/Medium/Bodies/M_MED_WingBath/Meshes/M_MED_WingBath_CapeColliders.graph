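# Physics asset for the male medium WingBath player body, named as the collision shapes for its cape.
# A head sphere and capsules on the neck, clavicles, spine and pelvis form the torso volume, while arm and leg bodies have no shapes.
# Capsule chains also cover the holster, left skirt panels and right skirt flap, each rooted at the pelvis.
# All constraints leave swing and twist free, so the listed angle limits are inactive.
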PhysicsAsset /BRCosmetics/Characters/Player/Male/Medium/Bodies/M_MED_WingBath/Meshes/M_MED_WingBath_CapeColliders
bodies {
 calf_l []
 calf_r []
 clavicle_l [{shapeType=Capsule, length=9.0192, radius=10, center=<6 -1 -2>, rotation=<rot 79.3028 -1.4096 174.1849>}]
 clavicle_r [{shapeType=Capsule, length=9.0192, radius=10, center=<-6 1 2>, rotation=<rot 79.3028 -1.4099 -5.8154>}]
 dyn_holster_driver [{shapeType=Capsule, length=0.505, radius=0.505, shapeName=dyn_holster_driver_capsule, rotation=<rot 0 -90 0>}]
 dyn_holster_main [{
   shapeType=Capsule
   length=18.0781
   radius=12
   shapeName=dyn_holster_main_capsule
   center=<0.5862 4.0909 -2.1988>
   rotation=<rot 0 81.8973 -89.902>
  }]
 dyn_holster_master [{shapeType=Capsule, length=0.505, radius=0.505, shapeName=dyn_holster_master_capsule, rotation=<rot 0 -90 0>}]
 dyn_holster_top_master [{shapeType=Capsule, length=0.505, radius=0.505, shapeName=dyn_holster_top_master_capsule, rotation=<rot 0 -90 0>}]
 dyn_skirt_back_01_l [{shapeType=Capsule, length=6, radius=6, shapeName=dyn_skirt_back_01_l_capsule, center=<6 0 0>, rotation=<rot 0 90 90>}]
 dyn_skirt_back_02_l [{shapeType=Capsule, length=6, radius=6, shapeName=dyn_skirt_back_02_l_capsule, center=<6 0 0>, rotation=<rot 0 90 90>}]
 dyn_skirt_back_03_l [{shapeType=Capsule, length=6, radius=6, shapeName=dyn_skirt_back_03_l_capsule, center=<6 0 0>, rotation=<rot 0 90 90>}]
 dyn_skirt_back_Base_l [{shapeType=Capsule, length=0.505, radius=0.505, shapeName=dyn_skirt_back_Base_l_capsule, rotation=<rot 0 -90 0>}]
 dyn_skirt_flap_01_r [{shapeType=Capsule, length=6, radius=6, shapeName=dyn_skirt_flap_01_r_capsule, center=<6 0 0>, rotation=<rot 0 90 90>}]
 dyn_skirt_flap_02_r [{shapeType=Capsule, length=6, radius=6, shapeName=dyn_skirt_flap_02_r_capsule, center=<6 0 0>, rotation=<rot 0 90 90>}]
 dyn_skirt_flap_03_r [{shapeType=Capsule, length=6, radius=6, shapeName=dyn_skirt_flap_03_r_capsule, center=<6 0 0>, rotation=<rot 0 90 90>}]
 dyn_skirt_front_01_l [{shapeType=Capsule, length=6, radius=6, shapeName=dyn_skirt_front_01_l_capsule, center=<6 0 0>, rotation=<rot 0 90 90>}]
 dyn_skirt_front_02_l [{shapeType=Capsule, length=6, radius=6, shapeName=dyn_skirt_front_02_l_capsule, center=<6 0 0>, rotation=<rot 0 90 90>}]
 dyn_skirt_front_03_l [{shapeType=Capsule, length=6, radius=6, shapeName=dyn_skirt_front_03_l_capsule, center=<6 0 0>, rotation=<rot 0 90 90>}]
 dyn_skirt_front_Base_l [{shapeType=Capsule, length=0.505, radius=0.505, shapeName=dyn_skirt_front_Base_l_capsule, rotation=<rot 0 -90 0>}]
 dyn_skirt_master_r [{shapeType=Capsule, length=0.505, radius=0.505, shapeName=dyn_skirt_master_r_capsule, rotation=<rot 0 -90 0>}]
 dyn_skirt_side_01_l [{shapeType=Capsule, length=6, radius=6, shapeName=dyn_skirt_side_01_l_capsule, center=<6 0 0>, rotation=<rot 0 90 90>}]
 dyn_skirt_side_02_l [{shapeType=Capsule, length=6, radius=6, shapeName=dyn_skirt_side_02_l_capsule, center=<6 0 0>, rotation=<rot 0 90 90>}]
 dyn_skirt_side_03_l [{shapeType=Capsule, length=6, radius=6, shapeName=dyn_skirt_side_03_l_capsule, center=<6 0 0>, rotation=<rot 0 90 90>}]
 dyn_skirt_side_Base_l [{shapeType=Capsule, length=0.505, radius=0.505, shapeName=dyn_skirt_side_Base_l_capsule, rotation=<rot 0 -90 0>}]
 foot_l []
 foot_r []
 head [{shapeType=Sphere, radius=17.3125, center=<11.028 -0.174 0>}]
 lowerarm_l []
 lowerarm_r []
 neck_02 [{shapeType=Capsule, length=13.4375, radius=7.3175, center=<-0.2642 -0.3408 0>, rotation=<rot 90 -171.1193 -175.3978>}]
 pelvis [{shapeType=Capsule, length=8.6456, radius=15, center=<-2.4887 0.1881 0>}]
 spine_01 [{shapeType=Capsule, length=7.9438, radius=15, center=<-1.7965 0.1884 0>}]
 spine_02 [
  {shapeType=Capsule, length=7.3264, radius=13.6322, center=<-1.711 -0.1851 0.0001>}
  {shapeType=Capsule, length=29.3956, radius=11.5807, center=<2.5024 0.4992 3.9825>, rotation=<rot 90 0.0049 0.0049>}
  {shapeType=Capsule, length=29.875, radius=11.1013, center=<2.4704 0.4991 -3.9825>, rotation=<rot 90 -0.0396 -0.0396>}
 ]
 spine_03 [{shapeType=Capsule, length=8.2107, radius=13.7031, center=<0.1445 0.2396 0.0001>}]
 spine_04 [{shapeType=Capsule, length=11, radius=14.1325, center=<2.766 0.7765 0.0002>}]
 spine_05 [{shapeType=Capsule, length=10, radius=15.695, center=<-2 1 0>}, {shapeType=Capsule, length=13.9518, radius=13.9425, center=<0 0.4452 0>}]
 thigh_l []
 thigh_r []
 upperarm_l []
 upperarm_r []
}
constraints [
 {
  bone1Name=spine_01
  bone2Name=pelvis
  swing1LimitDegrees=45
  swing1Motion=Free
  swing2LimitDegrees=45
  swing2Motion=Free
  twistLimitDegrees=45
  twistMotion=Free
 }
 {
  bone1Name=spine_02
  bone2Name=spine_01
  swing1LimitDegrees=45
  swing1Motion=Free
  swing2LimitDegrees=45
  swing2Motion=Free
  twistLimitDegrees=45
  twistMotion=Free
 }
 {
  bone1Name=spine_03
  bone2Name=spine_02
  swing1LimitDegrees=45
  swing1Motion=Free
  swing2LimitDegrees=45
  swing2Motion=Free
  twistLimitDegrees=45
  twistMotion=Free
 }
 {
  bone1Name=spine_04
  bone2Name=spine_03
  swing1LimitDegrees=45
  swing1Motion=Free
  swing2LimitDegrees=45
  swing2Motion=Free
  twistLimitDegrees=45
  twistMotion=Free
 }
 {
  bone1Name=spine_05
  bone2Name=spine_04
  swing1LimitDegrees=45
  swing1Motion=Free
  swing2LimitDegrees=45
  swing2Motion=Free
  twistLimitDegrees=45
  twistMotion=Free
 }
 {
  bone1Name=neck_02
  bone2Name=spine_05
  swing1LimitDegrees=45
  swing1Motion=Free
  swing2LimitDegrees=45
  swing2Motion=Free
  twistLimitDegrees=45
  twistMotion=Free
 }
 {
  bone1Name=head
  bone2Name=neck_02
  swing1LimitDegrees=45
  swing1Motion=Free
  swing2LimitDegrees=45
  swing2Motion=Free
  twistLimitDegrees=45
  twistMotion=Free
 }
 {
  bone1Name=thigh_l
  bone2Name=pelvis
  swing1LimitDegrees=45
  swing1Motion=Free
  swing2LimitDegrees=45
  swing2Motion=Free
  twistLimitDegrees=45
  twistMotion=Free
 }
 {
  bone1Name=foot_l
  bone2Name=calf_l
  swing1LimitDegrees=45
  swing1Motion=Free
  swing2LimitDegrees=45
  swing2Motion=Free
  twistLimitDegrees=45
  twistMotion=Free
 }
 {
  bone1Name=clavicle_r
  bone2Name=spine_05
  swing1LimitDegrees=45
  swing1Motion=Free
  swing2LimitDegrees=45
  swing2Motion=Free
  twistLimitDegrees=45
  twistMotion=Free
 }
 {
  bone1Name=upperarm_r
  bone2Name=clavicle_r
  swing1LimitDegrees=45
  swing1Motion=Free
  swing2LimitDegrees=45
  swing2Motion=Free
  twistLimitDegrees=45
  twistMotion=Free
 }
 {
  bone1Name=lowerarm_r
  bone2Name=upperarm_r
  swing1LimitDegrees=45
  swing1Motion=Free
  swing2LimitDegrees=45
  swing2Motion=Free
  twistLimitDegrees=45
  twistMotion=Free
 }
 {
  bone1Name=foot_r
  bone2Name=calf_r
  swing1LimitDegrees=45
  swing1Motion=Free
  swing2LimitDegrees=45
  swing2Motion=Free
  twistLimitDegrees=45
  twistMotion=Free
 }
 {
  bone1Name=lowerarm_l
  bone2Name=upperarm_l
  swing1LimitDegrees=45
  swing1Motion=Free
  swing2LimitDegrees=45
  swing2Motion=Free
  twistLimitDegrees=45
  twistMotion=Free
 }
 {
  bone1Name=upperarm_l
  bone2Name=clavicle_l
  swing1LimitDegrees=45
  swing1Motion=Free
  swing2LimitDegrees=45
  swing2Motion=Free
  twistLimitDegrees=45
  twistMotion=Free
 }
 {
  bone1Name=clavicle_l
  bone2Name=spine_05
  swing1LimitDegrees=45
  swing1Motion=Free
  swing2LimitDegrees=45
  swing2Motion=Free
  twistLimitDegrees=45
  twistMotion=Free
 }
 {
  bone1Name=calf_r
  bone2Name=thigh_r
  swing1LimitDegrees=45
  swing1Motion=Free
  swing2LimitDegrees=45
  swing2Motion=Free
  twistLimitDegrees=45
  twistMotion=Free
 }
 {
  bone1Name=thigh_r
  bone2Name=pelvis
  swing1LimitDegrees=45
  swing1Motion=Free
  swing2LimitDegrees=45
  swing2Motion=Free
  twistLimitDegrees=45
  twistMotion=Free
 }
 {
  bone1Name=calf_l
  bone2Name=thigh_l
  swing1LimitDegrees=45
  swing1Motion=Free
  swing2LimitDegrees=45
  swing2Motion=Free
  twistLimitDegrees=45
  twistMotion=Free
 }
 {
  bone1Name=dyn_skirt_master_r
  bone2Name=pelvis
  swing1LimitDegrees=45
  swing1Motion=Free
  swing2LimitDegrees=45
  swing2Motion=Free
  twistLimitDegrees=45
  twistMotion=Free
 }
 {
  bone1Name=dyn_skirt_side_03_l
  bone2Name=dyn_skirt_side_02_l
  swing1LimitDegrees=45
  swing1Motion=Free
  swing2LimitDegrees=45
  swing2Motion=Free
  twistLimitDegrees=45
  twistMotion=Free
 }
 {
  bone1Name=dyn_skirt_flap_01_r
  bone2Name=dyn_skirt_master_r
  swing1LimitDegrees=45
  swing1Motion=Free
  swing2LimitDegrees=45
  swing2Motion=Free
  twistLimitDegrees=45
  twistMotion=Free
 }
 {
  bone1Name=dyn_skirt_flap_02_r
  bone2Name=dyn_skirt_flap_01_r
  swing1LimitDegrees=45
  swing1Motion=Free
  swing2LimitDegrees=45
  swing2Motion=Free
  twistLimitDegrees=45
  twistMotion=Free
 }
 {
  bone1Name=dyn_skirt_flap_03_r
  bone2Name=dyn_skirt_flap_02_r
  swing1LimitDegrees=45
  swing1Motion=Free
  swing2LimitDegrees=45
  swing2Motion=Free
  twistLimitDegrees=45
  twistMotion=Free
 }
 {
  bone1Name=dyn_skirt_side_Base_l
  bone2Name=pelvis
  swing1LimitDegrees=45
  swing1Motion=Free
  swing2LimitDegrees=45
  swing2Motion=Free
  twistLimitDegrees=45
  twistMotion=Free
 }
 {
  bone1Name=dyn_skirt_side_01_l
  bone2Name=dyn_skirt_side_Base_l
  swing1LimitDegrees=45
  swing1Motion=Free
  swing2LimitDegrees=45
  swing2Motion=Free
  twistLimitDegrees=45
  twistMotion=Free
 }
 {
  bone1Name=dyn_skirt_side_02_l
  bone2Name=dyn_skirt_side_01_l
  swing1LimitDegrees=45
  swing1Motion=Free
  swing2LimitDegrees=45
  swing2Motion=Free
  twistLimitDegrees=45
  twistMotion=Free
 }
 {
  bone1Name=dyn_skirt_front_03_l
  bone2Name=dyn_skirt_front_02_l
  swing1LimitDegrees=45
  swing1Motion=Free
  swing2LimitDegrees=45
  swing2Motion=Free
  twistLimitDegrees=45
  twistMotion=Free
 }
 {
  bone1Name=dyn_holster_top_master
  bone2Name=pelvis
  swing1LimitDegrees=45
  swing1Motion=Free
  swing2LimitDegrees=45
  swing2Motion=Free
  twistLimitDegrees=45
  twistMotion=Free
 }
 {
  bone1Name=dyn_holster_driver
  bone2Name=dyn_holster_top_master
  swing1LimitDegrees=45
  swing1Motion=Free
  swing2LimitDegrees=45
  swing2Motion=Free
  twistLimitDegrees=45
  twistMotion=Free
 }
 {
  bone1Name=dyn_holster_master
  bone2Name=dyn_holster_driver
  swing1LimitDegrees=45
  swing1Motion=Free
  swing2LimitDegrees=45
  swing2Motion=Free
  twistLimitDegrees=45
  twistMotion=Free
 }
 {
  bone1Name=dyn_holster_main
  bone2Name=dyn_holster_master
  swing1LimitDegrees=45
  swing1Motion=Free
  swing2LimitDegrees=45
  swing2Motion=Free
  twistLimitDegrees=45
  twistMotion=Free
 }
 {
  bone1Name=dyn_skirt_back_Base_l
  bone2Name=pelvis
  swing1LimitDegrees=45
  swing1Motion=Free
  swing2LimitDegrees=45
  swing2Motion=Free
  twistLimitDegrees=45
  twistMotion=Free
 }
 {
  bone1Name=dyn_skirt_back_01_l
  bone2Name=dyn_skirt_back_Base_l
  swing1LimitDegrees=45
  swing1Motion=Free
  swing2LimitDegrees=45
  swing2Motion=Free
  twistLimitDegrees=45
  twistMotion=Free
 }
 {
  bone1Name=dyn_skirt_back_02_l
  bone2Name=dyn_skirt_back_01_l
  swing1LimitDegrees=45
  swing1Motion=Free
  swing2LimitDegrees=45
  swing2Motion=Free
  twistLimitDegrees=45
  twistMotion=Free
 }
 {
  bone1Name=dyn_skirt_back_03_l
  bone2Name=dyn_skirt_back_02_l
  swing1LimitDegrees=45
  swing1Motion=Free
  swing2LimitDegrees=45
  swing2Motion=Free
  twistLimitDegrees=45
  twistMotion=Free
 }
 {
  bone1Name=dyn_skirt_front_Base_l
  bone2Name=pelvis
  swing1LimitDegrees=45
  swing1Motion=Free
  swing2LimitDegrees=45
  swing2Motion=Free
  twistLimitDegrees=45
  twistMotion=Free
 }
 {
  bone1Name=dyn_skirt_front_01_l
  bone2Name=dyn_skirt_front_Base_l
  swing1LimitDegrees=45
  swing1Motion=Free
  swing2LimitDegrees=45
  swing2Motion=Free
  twistLimitDegrees=45
  twistMotion=Free
 }
 {
  bone1Name=dyn_skirt_front_02_l
  bone2Name=dyn_skirt_front_01_l
  swing1LimitDegrees=45
  swing1Motion=Free
  swing2LimitDegrees=45
  swing2Motion=Free
  twistLimitDegrees=45
  twistMotion=Free
 }
]
