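# Physics asset for the WingBath male medium legs mesh (Blaster), defining bodies for the holster, a back blaster bone and the right leg.
# Capsules on pelvis and right thigh, calf and foot give body volume, with thigh, calf and the anchor bones rotation-locked, while the foot rotates freely.
# A holster chain runs from a locked top master through a tightly limited driver to a looser main body.
# The back blaster hangs off the pelvis with angular limits, and the right skirt master is a locked zero-radius sphere.
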PhysicsAsset /BRCosmetics/Characters/Player/Male/Medium/Bodies/M_MED_WingBath/Meshes/M_MED_WingBath_Legs_Blaster_Physics
bodies {
 calf_r [
  {shapeType=Capsule, length=21.2247, radius=7.8595, center=<11.0572 2.1068 -0.751>, rotation=<rot -0.0765 94.2096 90.6596>}
  {shapeType=Capsule, length=16.2539, radius=7.2423, shapeName=calf_r_capsule, center=<31.9769 0.8513 0>, rotation=<rot -0.0002 79.9977 90>}
 ]
 dyn_Blaster_back_02 [{
   shapeType=Capsule
   length=3
   radius=2.5
   shapeName=dyn_Blaster_back_02_capsule
   center=<4.2493 0.1963 -1.6964>
   rotation=<rot 81.5626 -58.8572 -62.3231>
  }]
 dyn_holster_driver [{
   shapeType=Capsule
   length=5.79
   radius=2.0474
   shapeName=dyn_bagDrive_01_capsule
   center=<5.032 0 -0.0023>
   rotation=<rot 0.0009 -89.9997 90.1755>
  }]
 dyn_holster_main [
  {shapeType=Capsule, length=25.1766, radius=4.6378, shapeName=dyn_bag_00_capsule, center=<1.4976 1.2147 3.8779>, rotation=<rot 0 91.6526 90>}
  {
   shapeType=Capsule
   length=14.0619
   radius=3.3922
   shapeName=dyn_holster_main_capsule
   center=<-11.2952 6.4666 4.6349>
   rotation=<rot -1.506 26.3962 89.226>
  }
 ]
 dyn_holster_top_master [{shapeType=Capsule, length=0.505, radius=0.505, shapeName=dyn_holster_top_master_capsule, rotation=<rot 0 -90 0>}]
 dyn_skirt_master_r [{shapeType=Sphere, radius=0, shapeName=dyn_skirt_master_r_sphere}]
 foot_r [{
   shapeType=Capsule
   length=10.4457
   radius=7.7736
   shapeName=foot_r_capsule
   center=<2.9019 -5.2824 0.869>
   rotation=<rot -0.3918 2.2746 -86.6329>
  }]
 pelvis [{shapeType=Capsule, length=7.9585, radius=13.0111, center=<2.5806 -1.8931 0>}]
 thigh_r [
  {shapeType=Capsule, length=29.5105, radius=12.1323, center=<18.0702 0.0003 -2.8263>, rotation=<rot -90 -90.3524 -92.3851>}
  {shapeType=Capsule, length=6.25, radius=9.5, center=<17.5102 -0.5933 0.3279>, rotation=<rot -90 0.0099 179.99>}
 ]
}
constraints [
 {
  bone1Name=thigh_r
  bone2Name=pelvis
  swing1LimitDegrees=45
  swing1Motion=Locked
  swing2LimitDegrees=45
  swing2Motion=Locked
  twistLimitDegrees=45
  twistMotion=Locked
 }
 {
  bone1Name=dyn_holster_top_master
  bone2Name=pelvis
  swing1LimitDegrees=45
  swing1Motion=Locked
  swing2LimitDegrees=45
  swing2Motion=Locked
  twistLimitDegrees=45
  twistMotion=Locked
 }
 {
  bone1Name=dyn_skirt_master_r
  bone2Name=pelvis
  swing1LimitDegrees=45
  swing1Motion=Locked
  swing2LimitDegrees=45
  swing2Motion=Locked
  twistLimitDegrees=45
  twistMotion=Locked
 }
 {bone1Name=dyn_Blaster_back_02, bone2Name=pelvis, swing1LimitDegrees=20, swing2LimitDegrees=20, twistLimitDegrees=15}
 {bone1Name=dyn_holster_driver, bone2Name=dyn_holster_top_master, swing1LimitDegrees=5, swing2LimitDegrees=10, twistLimitDegrees=5}
 {bone1Name=dyn_holster_main, bone2Name=dyn_holster_driver, swing1LimitDegrees=22, swing2LimitDegrees=30, twistLimitDegrees=15}
 {
  bone1Name=calf_r
  bone2Name=thigh_r
  swing1LimitDegrees=45
  swing1Motion=Locked
  swing2LimitDegrees=45
  swing2Motion=Locked
  twistLimitDegrees=45
  twistMotion=Locked
 }
 {
  bone1Name=foot_r
  bone2Name=calf_r
  swing1LimitDegrees=45
  swing1Motion=Free
  swing2LimitDegrees=45
  swing2Motion=Free
  twistLimitDegrees=45
  twistMotion=Free
 }
]
