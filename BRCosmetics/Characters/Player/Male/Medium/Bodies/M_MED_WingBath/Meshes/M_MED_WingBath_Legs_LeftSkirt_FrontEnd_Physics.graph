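# Physics asset for the left skirt of the male medium WingBath cosmetic body, apparently the front-end variant.
# Front, side and back chains of three box bodies hang from base bones locked to the pelvis, with angular limits along each chain.
# Cross-constraints tie the chain tips: side to back with free rotation, front to side locked.
# Pelvis and left-leg capsules plus a wide, thin root box provide body and ground-like collision shapes.
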PhysicsAsset /BRCosmetics/Characters/Player/Male/Medium/Bodies/M_MED_WingBath/Meshes/M_MED_WingBath_Legs_LeftSkirt_FrontEnd_Physics
bodies {
 calf_l [{shapeType=Capsule, length=32.6849, radius=11.6844, center=<-20.2002 0 0.9072>, rotation=<rot 90 0 0>}]
 dyn_skirt_back_01_l [{shapeType=Box, extentX=14, extentY=2, extentZ=8, center=<6.637 0 0>, rotation=<rot 0 0.66 0>}]
 dyn_skirt_back_02_l [{shapeType=Box, extentX=12, extentY=3, extentZ=9, center=<8.2749 0 0>, rotation=<rot 0 0 -5.0548>}]
 dyn_skirt_back_03_l [{shapeType=Box, extentX=17, extentY=3, extentZ=11, center=<9.0738 -0.8086 -0.2853>, rotation=<rot -2.7607 -7.8554 -19.2441>}]
 dyn_skirt_back_Base_l [{shapeType=Capsule, length=0, radius=0, shapeName=dyn_skirt_back_Base_l_capsule}]
 dyn_skirt_front_01_l [{
   shapeType=Box
   extentX=11.5
   extentY=2.5
   extentZ=9.5
   shapeName=dyn_skirt_front_01_l_box
   center=<8.0108 -1.4412 1.1794>
   rotation=<rot 4.1392 -5.0732 39.1109>
  }]
 dyn_skirt_front_02_l [{
   shapeType=Box
   extentX=12.5
   extentY=4.5
   extentZ=8.5
   shapeName=dyn_skirt_front_02_l_box
   center=<8.0931 -1.6831 1.2445>
   rotation=<rot -0.5892 0.7971 36.4747>
  }]
 dyn_skirt_front_03_l [{
   shapeType=Box
   extentX=16.5
   extentY=3.5
   extentZ=10.5
   shapeName=dyn_skirt_front_03_l_box
   center=<7.2181 -0.6899 1.5811>
   rotation=<rot -4.2615 2.6111 58.4629>
  }]
 dyn_skirt_front_Base_l [{shapeType=Capsule, length=0, radius=0, shapeName=dyn_skirt_front_Base_l_capsule}]
 dyn_skirt_side_01_l [{shapeType=Box, extentX=15, extentY=1, extentZ=8, center=<7.5553 0 0>, rotation=<rot -2.7089 0 0>}]
 dyn_skirt_side_02_l [{shapeType=Box, extentX=11, extentY=3, extentZ=9, center=<7.1305 0 0>}]
 dyn_skirt_side_03_l [{shapeType=Box, extentX=15, extentY=3, extentZ=12, center=<7.5059 0 0>}]
 dyn_skirt_side_Base_l [{shapeType=Capsule, length=0, radius=0, shapeName=dyn_skirt_side_Base_l_capsule}]
 foot_l [{shapeType=Capsule, length=12.7957, radius=8.1395, center=<-4.6409 5.1059 0>, rotation=<rot 90 180 87.7526>}]
 pelvis [{shapeType=Capsule, length=9.6354, radius=11.7125, center=<-4 0 0>}]
 root [{shapeType=Box, extentX=200, extentY=200, extentZ=1.01, shapeName=root_box}]
 thigh_l [
  {shapeType=Capsule, length=38.1983, radius=8.135, center=<-20.2067 0 0.2164>, rotation=<rot 88.29 179.959 179.2471>}
  {
   shapeType=Capsule
   length=21.5
   radius=7.5
   shapeName=thigh_l_capsule
   center=<-16.8809 1.8865 -0.721>
   rotation=<rot 87.5157 179.9988 177.2181>
  }
  {shapeType=Capsule, length=38.1983, radius=8.135, center=<-20.2067 -3.1934 0.2164>, rotation=<rot 88.29 179.9997 179.9997>}
 ]
}
constraints [
 {
  bone1Name=dyn_skirt_back_Base_l
  bone2Name=pelvis
  swing1LimitDegrees=10
  swing1Motion=Locked
  swing2LimitDegrees=10
  swing2Motion=Locked
  twistLimitDegrees=10
  twistMotion=Locked
 }
 {
  bone1Name=dyn_skirt_front_Base_l
  bone2Name=pelvis
  swing1LimitDegrees=10
  swing1Motion=Locked
  swing2LimitDegrees=10
  swing2Motion=Locked
  twistLimitDegrees=10
  twistMotion=Locked
 }
 {
  bone1Name=dyn_skirt_side_Base_l
  bone2Name=pelvis
  swing1LimitDegrees=10
  swing1Motion=Locked
  swing2LimitDegrees=10
  swing2Motion=Locked
  twistLimitDegrees=10
  twistMotion=Locked
 }
 {bone1Name=dyn_skirt_back_01_l, bone2Name=dyn_skirt_back_Base_l, swing1LimitDegrees=10, swing2LimitDegrees=0, twistLimitDegrees=0}
 {bone1Name=dyn_skirt_back_02_l, bone2Name=dyn_skirt_back_01_l, swing1LimitDegrees=50, swing2LimitDegrees=50, twistLimitDegrees=30}
 {bone1Name=dyn_skirt_back_03_l, bone2Name=dyn_skirt_back_02_l, swing1LimitDegrees=40, swing2LimitDegrees=60, twistLimitDegrees=30}
 {
  bone1Name=dyn_skirt_side_03_l
  bone2Name=dyn_skirt_back_03_l
  swing1LimitDegrees=45
  swing1Motion=Free
  swing2LimitDegrees=45
  swing2Motion=Free
  twistLimitDegrees=45
  twistMotion=Free
 }
 {bone1Name=dyn_skirt_front_01_l, bone2Name=dyn_skirt_front_Base_l, swing1LimitDegrees=40, swing2LimitDegrees=30, twistLimitDegrees=60}
 {bone1Name=dyn_skirt_front_02_l, bone2Name=dyn_skirt_front_01_l, swing1LimitDegrees=40, swing2LimitDegrees=40, twistLimitDegrees=60}
 {bone1Name=dyn_skirt_front_03_l, bone2Name=dyn_skirt_front_02_l, swing1LimitDegrees=45, swing2LimitDegrees=60, twistLimitDegrees=30}
 {
  bone1Name=dyn_skirt_front_03_l
  bone2Name=dyn_skirt_side_03_l
  swing1LimitDegrees=45
  swing1Motion=Locked
  swing2LimitDegrees=45
  swing2Motion=Locked
  twistLimitDegrees=45
  twistMotion=Locked
 }
 {bone1Name=dyn_skirt_side_01_l, bone2Name=dyn_skirt_side_Base_l, swing1LimitDegrees=10, swing2LimitDegrees=0, twistLimitDegrees=0}
 {bone1Name=dyn_skirt_side_02_l, bone2Name=dyn_skirt_side_01_l, swing1LimitDegrees=50, swing2LimitDegrees=50, twistLimitDegrees=30}
 {bone1Name=dyn_skirt_side_03_l, bone2Name=dyn_skirt_side_02_l, swing1LimitDegrees=45, swing2LimitDegrees=60, twistLimitDegrees=30}
]
